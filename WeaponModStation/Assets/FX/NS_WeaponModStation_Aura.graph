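# Niagara aura effect for the weapon mod station, built from burst-spawned particles that persist rather than expiring by lifetime.
# Aura renders a faint white mesh; Corona renders a tall blue sprite with a dynamic material parameter.
# Corona's camera-facing fade and waveform flicker modules are disabled, as is the Empty emitter.
# User.FXScale only feeds Corona's mesh scale, which its sprite renderer ignores; User.FX_WeaponCaseSize is unreferenced in the listed modules.
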NiagaraSystem /WeaponModStation/Assets/FX/NS_WeaponModStation_Aura
dependencies {usedDataInterfaces [&NiagaraDataInterfaceCamera], usedDynamicInputs [], usedModules [&SystemState]}
emitters {
 Aura {
  emitterUpdateScript [
   &EmitterState(Life Cycle Mode=::NewEnumerator1, Loop Behavior=::NewEnumerator1, Loop Duration Mode=::NewEnumerator1, UseLoopDelay=false)
   &SpawnBurst_Instantaneous(Use Spawn Probability=false, UseLoopCountLimit=true)
  ]
  particleSpawnScript [
   &InitializeParticle(
    Color=<rgba 1 1 1 0.06>
    Mesh Scale=<1.3 1 1>
    Mesh Scale Mode=::NewEnumerator4
    Position Offset=<0 3 0>
    UsePositionOffset=false
    Write Material Random=true
   )
   &InitialMeshOrientation(Mesh Orientation Mode=::NewEnumerator2, Use Rotation Vector=false) {enabled=false}
  ]
  particleUpdateScript [
   &ParticleState(Kill Particles When Lifetime Has Elapsed=false, Let Infinitely Lived Particles Die When Emitter Deactivates=true)
   &DynamicMaterialParameters(
    Index 0 Param 1=1
    Index 0 Param 2=0.5
    Index 0 Param 3=5
    Module.Index0Param0Write=true
    Module.Index0Param1Write=true
    Module.Index0Param2Write=true
    Module.Index0Param3Write=false
   ) {enabled=false}
  ]
  rendererClasses [&NiagaraMeshRendererProperties]
  renderers [{rendererClass=&NiagaraMeshRendererProperties, rendererIndex=0}]
 }
 Corona {
  emitterUpdateScript [
   &EmitterState(Life Cycle Mode=::NewEnumerator1, Loop Behavior=::NewEnumerator1, Loop Duration Mode=::NewEnumerator1, UseLoopDelay=false)
   &SpawnBurst_Instantaneous(Use Spawn Probability=false, UseLoopCountLimit=true)
   /Niagara/Noise/Waveform(
    EnableExponent1=false
    Global Input Position:NiagaraFloat=&RandomRangeFloat(Minimum=0, Maximum=0.1)
    GlobalClampMax=false
    GlobalClampMin=false
    UseInput1=false
    UsePhase1=false
    "[1] Frequency":NiagaraFloat=&RandomRangeFloat(Minimum=0.15, Maximum=0.25)
   ) {module=Waveform001}
  ]
  particleSpawnScript [
   &InitializeParticle(
    Color=<rgba 0.15 0.3625 1 1>
    Mesh Scale Mode=::NewEnumerator3
    Mesh Uniform Scale=&Multiply_Float(A=3, B=$User.FXScale)
    Position Mode=::NewEnumerator0
    Sprite Rotation Mode=::NewEnumerator0
    Sprite Size=<35 120>
    Sprite Size Mode=::NewEnumerator4
    Write Material Random=true
   )
   /SetVariables(Particles.SpriteFlareSize:Vector2f=&Multiply_Vector2D(A=$Particles.SpriteSize, B=<0.38 1.3>)) {enabled=false, isSetParameters=true}
   &SpriteFacingAndAlignment(Sprite Alignment=<1 0 0>, Sprite Facing=<0.45 0.69 0.3958>, WriteAlignment=true, WriteFacing=true) {enabled=false}
   &SpriteFacingAndAlignment(Sprite Alignment=<1 0 0>, Sprite Facing=<0.7068 0.07 0.3>, WriteAlignment=true, WriteFacing=false) {module=SpriteFacingAndAlignment001}
  ]
  particleUpdateScript [
   &ParticleState(Kill Particles When Lifetime Has Elapsed=false, Let Infinitely Lived Particles Die When Emitter Deactivates=true)
   &UpdateMeshOrientation(
    Facing Direction=&ReturnCameraVector(Camera Vectors=::NewEnumerator4, Coordinate Space=::Simulation)
    Facing Interpolation=false
    Orientation Method=::NewEnumerator1
    Override Quat=false
    Stabilization Mode:ENiagara_OrientVectorCount=::NewEnumerator0
    TransformFacing=false
   ) {enabled=false}
   /SetVariables(Particles.FacingCamera:NiagaraFloat=&ClampFloat(Float=&RemapRange(
      Input Value To Remap=&DotProduct(
       Vector A=&NormalizeVector(
        Vector To Normalize=&Multiply_Vector(A=&ReturnCameraVector(Camera Vectors=::NewEnumerator4, Coordinate Space=::Simulation), B=<1 1 0>)
        Safe Normalize=false
        Safe Fallback Threshold=0
        Safe Fallback Vector=<0 0 0.001>
       )
       Vector B=<0 1 0>
      )
      Input Min=0.25
      Input Max=0.6
      Output Min=0
      Output Max=1
     ))) {enabled=false, isSetParameters=true}
   &CameraOffset(Camera Offset Amount=4) {enabled=false}
   &DynamicMaterialParameters(
    Index 0 Param 1=2
    Module.Index0Param0Write=true
    Module.Index0Param1Write=false
    Module.Index0Param2Write=false
    Module.Index0Param3Write=false
   ) {module=DynamicMaterialParameters001}
   &ScaleColor(
    Scale Alpha=&Multiply_Float(A=$Particles.FacingCamera, B=&Lerp_Float(A=0.25, B=0.6, Alpha=$Emitter.Waveform001.Waveform))
    ScaleA=true
    ScaleRGB=true
   ) {enabled=false, module=ScaleColor001}
   /SetVariables(Particles.ColorFlare:LinearColor=&MakeLinearColorFromVectorAndFloat(
     "Vector (RGB)"=&MakeVectorFromLinearColorRGB(Color=$Particles.Color)
     "Float (Alpha)"=&Lerp_Float(A=0.3, B=0.5, Alpha=$Emitter.Waveform001.Waveform)
    )) {enabled=false, isSetParameters=true}
  ]
  rendererClasses [&NiagaraSpriteRendererProperties]
  renderers [{rendererClass=&NiagaraSpriteRendererProperties}]
 }
 Empty {
  bEnabled=false
  emitterUpdateScript [&EmitterState()]
  particleSpawnScript [&InitializeParticle(Color Mode=::NewEnumerator0, Lifetime=5, UsePositionOffset=false, Write Material Random=true)]
  particleUpdateScript [&ParticleState()]
  rendererClasses [&NiagaraSpriteRendererProperties]
  renderers [{rendererClass=&NiagaraSpriteRendererProperties}]
 }
}
summary {
 systemName=NS_WeaponModStation_Aura
 emitters [
  {emitterName=Corona, rendererClasses [&NiagaraSpriteRendererProperties]}
  {emitterName=Aura, rendererClasses [&NiagaraMeshRendererProperties]}
  {bEnabled=false, emitterName=Empty, rendererClasses [&NiagaraSpriteRendererProperties]}
 ]
 userVariables [
  User.FX_WeaponCaseSize {type=NiagaraInt32, defaultValue=1}
  User.FXScale {type=NiagaraFloat, defaultValue=1, description=""}
 ]
}
