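# Translucent lit glass material for the weapon mod bench that fakes light falling on surfaces seen through it.
# A scene-minus-pixel depth falloff with DepthFade scales tinted light, multiplied by scene color or diffuse (selected by actor height), with a simpler mobile path and eye-adaptation compensation.
# A bounding-box UVW sphere mask defines a shape.
# Opacity math combines Fresnel with a camera-depth term that LightOff custom primitive data dims.
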Material /WeaponModStation/Assets/Materials/M_Bench_WeaponModGlass
graph {
 properties {blendMode=BLEND_Translucent, pixelDepthOffsetMode=PDOM_Legacy, translucencyLightingMode=TLM_Surface, usedShadingModels=Default Lit}
 nodes [
  Constant() {r=0.2}
  Constant()
  n1 = Subtract(A=PixelDepth().R, B=ScalarParameter() {defaultValue=1000, parameterName=OpacityLerpStart, sliderMax=200})
  n2 = ScalarParameter() {defaultValue=500, parameterName=OpacityLerpDistance, sliderMax=1000}
  n3 = LinearInterpolate(A=NamedRerouteUsage(), Alpha=Saturate(None=Divide(A=@n1, B=@n2))) {constAlpha=0, constB=0}
  n4 = OneMinus(None=ScalarParameter() {bUseCustomPrimitiveData=true, parameterName=LightOff})
  n5 = ScalarParameter() {defaultValue=0.12, parameterName=OpcaityMult2}
  n6 = ScalarParameter() {defaultValue=0.2, parameterName=OpcaityMult}
  n7 = NamedRerouteUsage()
  Max(
   A=Fresnel(ExponentIn=ScalarParameter() {defaultValue=2, parameterName=Opacity Exponent})
   B=Multiply(A=Multiply(A=@n3, B=@n4), B=LinearInterpolate(A=@n5, B=@n6, Alpha=@n7))
  )
  n8 = ScalarParameter() {defaultValue=500, parameterName=LightStrength}
  n9 = NamedRerouteUsage()
  n10 = Subtract(A=SceneDepth().R, B=PixelDepth().R)
  n11 = ScalarParameter() {defaultValue=150, parameterName=LightAttenuation}
  n12 = DepthFade(
   Opacity=OneMinus(None=Divide(A=@n10, B=@n11))
   FadeDistance=LinearInterpolate(A=ScalarParameter() {defaultValue=10, parameterName=DepthFadeDistance}, Alpha=NamedRerouteUsage()) {constA=10, constB=30}
  )
  n13 = Multiply(
   A=VectorParameter() {parameterName=LightColor, defaultValue=<rgba 0.5396 0.6182 0.8542 1>}.RGB
   B=NamedRerouteDeclaration(None=Saturate(None=@n12)) {name=Opacity_1, nodeColor=<rgba 1 0 0.8438 1>}
  )
  n14 = ComponentMask(None=SceneColor!(UVs).RGB) {b=true, g=true, r=true}
  n15 = SceneTexture() {sceneTextureId=PPI_DiffuseColor}
  n16 = ComponentMask(None=ActorPositionWS()) {b=true}
  n17 = LinearInterpolate(
   A=LinearInterpolate(A=@n14, B=ComponentMask(None=@n15.Color) {b=true, g=true, r=true}) {constAlpha=0}
   B=@n14
   Alpha=NamedRerouteDeclaration(None=Step(X=@n16) {constX=2500, constY=2500}) {name=PositionMask, nodeColor=<rgba 0.9688 1 0 1>}
  )
  n18 = EyeAdaptation()
  n19 = Multiply(
   A=Multiply(A=LinearInterpolate(B=@n8, Alpha=@n9), B=@n13)
   B=LinearInterpolate(A=@n17, B=Multiply(A=@n17, B=@n18.EyeAdaptation)) {constAlpha=1}
  )
  n20 = Multiply(A=LinearInterpolate(B=ScalarParameter() {defaultValue=25, parameterName=Mobile-LightStrength}, Alpha=@n9), B=@n13)
  NamedRerouteDeclaration(None=EyeAdaptationInverse(
    LightValueInput=ShadingPathSwitch(Default=@n19, Mobile=@n20)
    AlphaInput=ScalarParameter() {defaultValue=1, parameterName=EyeAdaptionInverse}
   ).EyeAdaptationInverse) {name=Emissive, nodeColor=<rgba 0 0.625 1 1>}
  n21 = MaterialFunctionCall() {materialFunction=&BoundingBoxBased_0-1_UVW}
  n22 = ConstantBiasScale(None=InverseLinearInterpolate(
    A=ScalarParameter() {parameterName=RemapMinY}
    B=ScalarParameter() {defaultValue=0.8543, parameterName=RemapMaxY}
    Value=Clamp(None=@n21.B) {maxDefault=0.9, minDefault=0.1}
   )) {bias=-0.5, scale=2}
  n23 = ScalarParameter() {defaultValue=0.15, parameterName=RemapMinX}
  n24 = ScalarParameter() {defaultValue=0.85, parameterName=RemapMaxX}
  n25 = Clamp(None=@n21.R) {maxDefault=0.9, minDefault=0.1}
  n26 = MaterialFunctionCall(Float2=AppendVector(A=OneMinus(None=@n22), B=InverseLinearInterpolate(A=@n23, B=@n24, Value=@n25))) {materialFunction=&BreakOutFloat2Components}
  n27 = Multiply(A=Add(A=@n26.R) {constB=0.7}) {constB=-0.8}
  n28 = OneMinus(None=Abs(None=ConstantBiasScale(None=@n26.G) {bias=-0.5, scale=2}))
  NamedRerouteDeclaration(None=OneMinus(None=SphereMask(A=@n27, B=@n28) {attenuationRadius=4, hardnessPercent=25})) {name=Shape, nodeColor=<rgba 0.4297 1 0 1>}
  NamedRerouteUsage()
  ScalarParameter() {defaultValue=1, parameterName=Specular}
  ScalarParameter() {defaultValue=0.95, parameterName=Metallic}
 ]
}
