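# Translucent two-sided material that fakes a tinted light glow for the weapon mod station's Paprika weapon case padding.
# A UV-derived shape with panning noise masks a depth-attenuated, depth-faded light, multiplied by scene color and eye-adaptation compensated, with a simpler mobile path.
# Custom primitive data LightOff dims it, and pixel depth fades the shape with distance.
# Offset math rebuilds local vertices on a camera-derived, Z-up basis, resembling an axis-locked billboard.
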Material /WeaponModStation/Assets/Materials/M_PaprikaWeaponCase_Padding_Light
graph {
 properties {blendMode=BLEND_Translucent, pixelDepthOffsetMode=PDOM_Legacy, twoSided=true, usedShadingModels=Default Lit}
 nodes [
  n1 = OneMinus(None=ScalarParameter() {bUseCustomPrimitiveData=true, parameterName=LightOff})
  Clamp(None=@n1) {minDefault=0.5}
  n2 = MaterialFunctionCall(Float2=TextureCoordinate()) {materialFunction=&BreakOutFloat2Components}
  n3 = Abs(None=ConstantBiasScale(None=@n2.R) {bias=-0.5, scale=2})
  n4 = MaterialFunctionCall(Float2=AppendVector(A=OneMinus(None=@n3), B=@n2.G)) {materialFunction=&BreakOutFloat2Components}
  n5 = Multiply(A=Add(A=@n4.R) {constB=0.7}) {constB=-0.8}
  n6 = OneMinus(None=Abs(None=ConstantBiasScale(None=@n4.G) {bias=-0.5, scale=2}))
  n7 = Power(Base=Clamp(None=OneMinus(None=@n4.R)) {minDefault=0.1}, Exp=ScalarParameter() {defaultValue=0.5, parameterName=Param_1}) {constExponent=2.5}
  n8 = Multiply(A=Panner(Coordinate=TextureCoordinate()) {speedX=0.1, speedY=0.1}, B=ScalarParameter() {defaultValue=3, parameterName=NoiseTiling})
  n9 = Multiply(A=ConstantBiasScale(None=TextureSample(UVs=@n8) {samplerType=SAMPLERTYPE_LinearGrayscale, texture=/Game/Global/Textures/Default/Noise/EqualizerComponents_v2_Noise}.R) {bias=-0.5, scale=2}) {constB=0.02}
  n10 = Multiply(A=OneMinus(None=SphereMask(A=@n5, B=@n6) {attenuationRadius=4, hardnessPercent=25}), B=Saturate(None=Subtract(A=@n7, B=@n9)))
  n11 = Subtract(A=PixelDepth().R, B=ScalarParameter() {defaultValue=1000, parameterName=OpacityLerpStart, sliderMax=200})
  n12 = ScalarParameter() {defaultValue=500, parameterName=OpacityLerpDistance, sliderMax=1000}
  Multiply(
   A=LinearInterpolate(A=NamedRerouteDeclaration(None=@n10) {name=Shape, nodeColor=<rgba 0.4297 1 0 1>}, Alpha=Divide(A=@n11, B=@n12)) {constB=0}
   B=@n1
  )
  Multiply(A=@n1) {constB=0.5}
  Multiply(A=SkyLightEnvMapSample(Direction=ReflectionVectorWS()), B=Fresnel() {exponent=1}) {constB=0.3}
  n13 = ScalarParameter() {defaultValue=200, parameterName=LightStrength}
  n14 = NamedRerouteUsage()
  n15 = Subtract(A=SceneDepth().R, B=PixelDepth().R)
  n16 = ScalarParameter() {defaultValue=150, parameterName=LightAttenuation}
  n17 = DepthFade(Opacity=OneMinus(None=Divide(A=@n15, B=@n16)), FadeDistance=ScalarParameter() {defaultValue=20, parameterName=DepthFadeDistance})
  n18 = Multiply(
   A=VectorParameter() {parameterName=LightColor, defaultValue=<rgba 0.0842 0.483 1 1>}.RGB
   B=NamedRerouteDeclaration(None=Saturate(None=@n17)) {name=Opacity, nodeColor=<rgba 1 0 0.8438 1>}
  )
  n19 = SceneColor!(UVs)
  n20 = Multiply(
   A=Multiply(A=LinearInterpolate(B=@n13, Alpha=@n14), B=@n18)
   B=Multiply(A=ComponentMask(None=@n19.RGB) {b=true, g=true, r=true}, B=EyeAdaptation().EyeAdaptation)
  )
  n21 = Multiply(A=LinearInterpolate(B=ScalarParameter() {defaultValue=25, parameterName=Mobile-LightStrength}, Alpha=@n14), B=@n18)
  NamedRerouteDeclaration(None=EyeAdaptationInverse(
    LightValueInput=ShadingPathSwitch(Default=@n20, Mobile=@n21)
    AlphaInput=ScalarParameter() {defaultValue=1, parameterName=EyeAdaptionInverse}
   ).EyeAdaptationInverse) {name=Emissive, nodeColor=<rgba 0 0.625 1 1>}
  NamedRerouteUsage()
  n22 = MaterialFunctionCall(Float3=MaterialFunctionCall() {materialFunction=&LocalPosition}."Local Position (Excluding Offsets)") {materialFunction=&BreakOutFloat3Components}
  n23 = MaterialFunctionCall(X=@n22.G, Y=@n22.G, Z=@n22.G) {materialFunction=&MakeFloat3}
  n24 = ObjectPositionWS()
  n25 = CameraPositionWS()
  n26 = Add(A=Multiply(A=Subtract(A=@n24, B=@n25), B=ScalarParameter() {defaultValue=-500, parameterName=Param_9}) {constB=500})
  n27 = MaterialFunctionCall(Float2=@n26) {materialFunction=&BreakOutFloat2Components}
  n28 = Distance(A=@n27.R, B=Constant())
  n29 = ComponentMask(None=@n26) {g=true, r=true}
  n30 = AppendVector(A=Constant(), B=@n27.G)
  n31 = AppendVector(A=Normalize(VectorInput=If(A=@n28, "A > B"=@n29, "A == B"=@n30, "A < B"=@n30) {constB=7500}), B=Constant())
  n32 = Constant3Vector() {constant=<rgba 0 0 1 0>}
  n33 = CrossProduct(A=@n31, B=@n32)
  Subtract(
   A=TransformPosition(None=MaterialFunctionCall!(VectorToTransform=@n23.Result, BasisX=@n31, BasisY=@n33, BasisZ=@n32) {materialFunction=/Engine/Functions/Engine_MaterialFunctions02/Math/InverseTransformMatrix}.Transformed Vector 3) {transformType=TRANSFORMPOSSOURCE_World}
   B=WorldPosition() {worldPositionShaderOffset=WPT_ExcludeAllShaderOffsets}.XYZ
  )
 ]
}
